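# Anim Blueprint for the WWII Pilot Sci-Fi body cosmetic, copying the attached parent's pose and adding coat secondary motion.
# Additive bone offsets adjust back coat flaps for stationary crouching and skydiving, feeding RigidBody simulation, which is weakened when running fast.
# A final offset shifts the coat's split bone by right-thigh yaw relative to the pelvis.
# The update event computes these alphas, plus Jump_Alpha and Run_Rotate, which the AnimGraph never reads.
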AnimBlueprint /BRCosmetics/Characters/Player/Male/Medium/Bodies/M_MED_WWII_Pilot_Sci_Fi/Meshes/M_MED_WWII_Pilot_Sci_Fi_AnimBp
target_skeleton=&/M_MED_WWII_Pilot_Sci_Fi_Skeleton
graphs [
 AnimGraph {nodes [
   n1 = Get Crouching alpha() -> (Crouching alpha:double)
   n2 = "Transform (Modify) Bone - Bone: dyn_coat_base_back_l"(
    ComponentPose=Local To Component(LocalPose=Copy Pose From Mesh() {bUseAttachedParent=true})
    Alpha=@n1
    Translation="0,0,0"
    Rotation="0,0,80"
    Scale="1,1,1"
   ) {
    rotationMode=BMM_Additive
    rotationSpace=BCS_BoneSpace
    translationMode=BMM_Additive
    translationSpace=BCS_BoneSpace
    alphaBoolBlend {blendOption=Linear}
    alphaScaleBiasClamp {bInterpResult=true, interpSpeedDecreasing=30, interpSpeedIncreasing=30}
    boneToModify {boneName=dyn_coat_base_back_l}
    rotation {roll=80}
   }
   n3 = "Transform (Modify) Bone - Bone: dyn_coat_base_back_r"(ComponentPose=@n2, Alpha=@n1, Translation="0,0,0", Rotation="-50,0,-80", Scale="1,1,1") {
    rotationMode=BMM_Additive
    rotationSpace=BCS_BoneSpace
    translationMode=BMM_Additive
    translationSpace=BCS_BoneSpace
    alphaBoolBlend {blendOption=Linear}
    alphaScaleBiasClamp {bInterpResult=true, interpSpeedDecreasing=30, interpSpeedIncreasing=30}
    boneToModify {boneName=dyn_coat_base_back_r}
    rotation {pitch=-50, roll=-80}
   }
   n4 = Get Skydiving alpha() -> (Skydiving alpha:double)
   n5 = "Transform (Modify) Bone - Bone: dyn_coat_base_back_l"(ComponentPose=@n3, Alpha=@n4, Translation="0,0,0", Rotation="35,0,0", Scale="1,1,1") {
    rotationMode=BMM_Additive
    rotationSpace=BCS_BoneSpace
    translationMode=BMM_Additive
    translationSpace=BCS_BoneSpace
    alphaBoolBlend {blendOption=Linear}
    alphaScaleBiasClamp {interpSpeedDecreasing=30, interpSpeedIncreasing=30}
    boneToModify {boneName=dyn_coat_base_back_l}
    rotation {pitch=35}
   }
   n6 = "Transform (Modify) Bone - Bone: dyn_coat_base_back_r"(ComponentPose=@n5, Alpha=@n4, Translation="0,0,0", Rotation="-30,0,0", Scale="1,1,1") {
    rotationMode=BMM_Additive
    rotationSpace=BCS_BoneSpace
    translationMode=BMM_Additive
    translationSpace=BCS_BoneSpace
    alphaBoolBlend {blendOption=Linear}
    alphaScaleBiasClamp {interpSpeedDecreasing=30, interpSpeedIncreasing=30}
    boneToModify {boneName=dyn_coat_base_back_r}
    rotation {pitch=-30}
   }
   n7 = RigidBody(ComponentPose=@n6, Alpha=Get Coat_Running_Alpha() -> (Coat_Running_Alpha:double)) {
    bClampLinearTranslationLimitToRefPose=true
    bOverrideWorldGravity=true
    overridePhysicsAsset=&/M_MED_WWII_Pilot_Sci_Fi_Physics
    simulationSpace=BaseBoneSpace
    alphaBoolBlend {blendOption=Linear}
    baseBoneRef {boneName=root}
    componentAppliedLinearAccClamp=<6000 6000 6000>
    componentLinearAccScale=<0.2 0.2 0.2>
    componentLinearVelScale=<0.2 0.2 0.2>
    overrideWorldGravity {z=-980}
   }
   n8 = "Transform (Modify) Bone - Bone: dyn_parent_split"(ComponentPose=@n7, Alpha=Get Leg_Rot_Forward() -> (Leg_Rot_Forward:double), Translation="0,-20,5", Rotation="0,0,0", Scale="1,1,1") {
    rotationMode=BMM_Additive
    rotationSpace=BCS_ParentBoneSpace
    translationMode=BMM_Additive
    translationSpace=BCS_ParentBoneSpace
    alphaBoolBlend {blendOption=Linear}
    boneToModify {boneName=dyn_parent_split}
    translation {y=-20, z=5}
   }
   Output Pose(Result=Component To Local(ComponentPose=@n8)) {layerGroup=None}
  ]}
 EventGraph {nodes [
   Event BlueprintUpdateAnimation()
   n1 = NOT Boolean(A=Get bIsSkyDiving())
   n2 = "float >= float"(A=Get PawnSpeed2D(), B=412) {is=title K2Node_CallFunction}
   > Set Coat_Running_Alpha(Coat_Running_Alpha:double=SelectFloat(A=0.65, B=1, bPickA=AND Boolean(A=@n2, B=@n1))) -> (Output_Get:double)
   > Set Jump_Alpha(Jump_Alpha:double=SelectFloat(A=0.5, B=1, bPickA=Get bIsJumping())) -> (Output_Get:double)
   n3 = BreakTransform(InTransform=GetDeltaTransformFromRefPose(self=GetOwnerSkeletalMeshComponent(), BoneName=thigh_r, BaseName=pelvis))
   > Set Leg_Rot_Forward(Leg_Rot_Forward:double=MapRangeClamped(Value=BreakRotator(InRot=@n3.Rotation).Yaw, InRangeA=0, InRangeB=-10)) -> (Output_Get:double)
   n4 = Get bIsTargeting()
   > Set Run_Rotate(Run_Rotate:double="To Float (Boolean)"(InBool=@n4)) -> (Output_Get:double)
   n5 = NOT Boolean(A=Get bIsMoving2D())
   n6 = Get bIsCrouching()
   > Set Crouching alpha(Crouching alpha:double=SelectFloat(A=1, B=0, bPickA=AND Boolean(A=@n6, B=@n5))) -> (Output_Get:double)
   > Set Skydiving alpha(Skydiving alpha:double=SelectFloat(A=1, B=0, bPickA=Get bIsSkyDiving())) -> (Output_Get:double)
   TryGetPawnOwner()
   "To Float (Boolean)"(InBool=NOT Boolean(A=@n4))
  ]}
]
variables [
 Coat_Running_Alpha {type=number}
 Crouching alpha {type=number}
 Jump_Alpha {type=number}
 Leg_Rot_Forward {type=number}
 Run_Rotate {type=number}
 Skydiving alpha {type=number}
]
